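# Animation Blueprint for the male medium Quill head cosmetic mesh, driving it from a source body mesh.
# On initialization it caches the owner's skeletal mesh and whether the playlist uses custom character parts.
# It copies that mesh's pose, runs it through the head's Control Rig and caches the result.
# When custom parts are used, an instant bool blend selects a version with the head bone shrunk to near zero.
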AnimBlueprint /BRCosmetics/Characters/Player/Male/Medium/Bodies/M_MED_Quill/Meshes/M_MED_Quill_Head_01_AnimBlueprint
target_skeleton=&/M_MED_Quill_Head_01_Skeleton
graphs [
 AnimGraph {nodes [
   n1 = "Transform (Modify) Bone - Bone: head"!(ComponentPose=Local To Component(LocalPose=Use cached pose 'PreScale Cache'()), Alpha=1) -> (Pose) {scaleMode=BMM_Replace, alphaBoolBlend {blendOption=Linear}, boneToModify {boneName=head}, scale=<0.01 0.01 0.01>}
   Output Pose(Result=Blend Poses by bool(
     BlendPose_0=Component To Local(ComponentPose=@n1)
     BlendPose_1=Use cached pose 'PreScale Cache'()
     BlendTime_0=0
     BlendTime_1=0
     bActiveValue=Get GetPlayerListUsesCustomCharacterParts() -> (GetPlayerListUsesCustomCharacterParts:bool)
    ) {blendType=Linear}) {layerGroup=None}
   n2 = Control Rig(Source=Copy Pose From Mesh(SourceMeshComponent=Get MeshToCopy() -> (MeshToCopy:Skeletal Mesh Component Object Reference))) {
    alphaBoolBlend {blendOption=Linear}
    controlRigAssetReference {blueprintRigClass=&/M_MED_Quill_Head_01_CtrlRig.M_MED_Quill_Head_01_CtrlRig_C}
   }
   Save cached pose 'PreScale Cache'(Pose=@n2) {cacheName=PreScale Cache}
  ]}
 EventGraph {nodes [
   Event BlueprintInitializeAnimation()
   > Set MeshToCopy(MeshToCopy:Skeletal Mesh Component Object Reference=GetOwnerSkeletalMeshComponent()) -> (Output_Get:Skeletal Mesh Component Object Reference)
   > Set GetPlayerListUsesCustomCharacterParts(GetPlayerListUsesCustomCharacterParts:bool=GetPlaylistUsesCustomCharacterParts!() -> (ReturnValue:bool)) -> (Output_Get:bool)
   Event BlueprintUpdateAnimation()
  ]}
]
variables [
 GetPlayerListUsesCustomCharacterParts {type=boolean}
 MeshToCopy {type=SkeletalMeshComponent}
]
